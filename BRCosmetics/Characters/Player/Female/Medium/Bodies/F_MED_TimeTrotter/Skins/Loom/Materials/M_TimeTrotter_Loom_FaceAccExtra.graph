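# Character material for the TimeTrotter Loom skin's extra face accessory, adding an animated emissive gradient to the 2023 character parent.
# Parent PreFX attributes pass through switchable hair and flyaway-hair functions, then blend with a black-base emissive layer before PostFX.
# TechArtMap red and green build the emissive mask, while blue plus time scrolls a gradient lookup, using plain-UV derivatives for mip selection.
# Desaturation, brightness and a Fresnel term with adjustable inversion and power shape the glow.
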Material /BRCosmetics/Characters/Player/Female/Medium/Bodies/F_MED_TimeTrotter/Skins/Loom/Materials/M_TimeTrotter_Loom_FaceAccExtra
graph {
 properties {
  bUseMaterialAttributes=true
  bUsedWithClothing=true
  bUsedWithInstancedStaticMeshes=true
  bUsedWithMeshDeformer=true
  bUsedWithMorphTargets=true
  bUsedWithNanite=true
  bUsedWithSkeletalMesh=true
  pixelDepthOffsetMode=PDOM_Legacy
  shadingModel=MSM_FromMaterialExpression
  usedShadingModels=Default Lit | Subsurface | Clear Coat | Subsurface Profile | Cloth
 }
 nodes [
  n1 = NamedRerouteUsage()
  n2 = Constant()
  n3 = NamedRerouteUsage()
  n4 = LinearInterpolate(A=OneMinus(None=@n3), B=@n3, Alpha=ScalarParameter() {group=TimeTrotter, parameterName=Emissive_Fresnel, sortPriority=39})
  n5 = ScalarParameter() {defaultValue=2, group=TimeTrotter, parameterName=Emissive_FresnelPower, sortPriority=39}
  n6 = Multiply(A=NamedRerouteUsage(), B=Saturate(None=Power(Base=@n4, Exp=@n5)))
  n7 = MaterialFunctionCall() {materialFunction=&MF_CharacterParent_2023_PreFX}
  MaterialFunctionCall(
   Material=BlendMaterialAttributes(
    A=NamedRerouteUsage()
    B=SetMaterialAttributes!(MaterialAttributes=@n1, Base Color=@n2, Emissive Color=@n6) {attributeSetTypes [#1, #2]}
    Alpha=NamedRerouteUsage()
   ) {vertexAttributeBlendType=UseA}
   M=@n7.M
   Pre Skinned Local Position=@n7.M
   Distance Blend=@n7.M
   Fresnel=@n7.M
  ) {materialFunction=&MF_CharacterParent_2023_PostFX}
  NamedRerouteDeclaration(None=@n7.Fresnel) {name=Fresnel, nodeColor=<rgba 1 0 0.7734 1>}
  n8 = MaterialFunctionCall!(In=@n7.Result, Mask Input=@n7.Result) {materialFunction=/Game/Global/Materials/Character_Parent_2023/Hair_Parent_2023/MaterialFunctions/MF_FN_Hair_2023}
  n9 = BlendMaterialAttributes(
   A=@n7.Result
   B=StaticSwitchParameter(True=@n8.Result, False=@n7.Result) {defaultValue=true, group=Skin FX, parameterName=UseFNHair2_0}
   Alpha=@n8.HairMask
  )
  NamedRerouteDeclaration(None=StaticSwitchParameter(
    True=MaterialFunctionCall!(MaterialInput=@n9) {materialFunction=/Game/Global/Materials/Character_Parent_2023/Hair_Parent_2023/MaterialFunctions/MF_FlyAwayHair_2023}.Result
    False=@n9
   ) {group=Skin FX, parameterName=IsFlyAwayHair?}) {name=Material Attributes, nodeColor=<rgba 1 0.5156 0 1>}
  n10 = TextureSampleParameter2D() {
   group=TimeTrotter
   parameterName=TechArtMap
   samplerSource=SSM_Wrap_WorldGroupSettings
   samplerType=SAMPLERTYPE_Masks
   texture=&T_BlackMask
  }
  n11 = Multiply(A=Time(), B=ScalarParameter() {defaultValue=0.5, group=TimeTrotter, parameterName=EmissiveGradient.PanningSpeed})
  n12 = TextureCoordinate()
  n13 = TextureSampleParameter2D!(UVs=Frac(None=Add(A=@n10.B, B=@n11)), "DDX(UVs)"=DDX(Value=@n12), "DDY(UVs)"=DDY(Value=@n12), Apply View MipBias) {
   group=TimeTrotter
   mipValueMode=TMVM_Derivative
   parameterName=PanningColorGradient
   samplerSource=SSM_Wrap_WorldGroupSettings
   sortPriority=6
   texture=/Game/Global/Textures/Default/Masks_and_Gradients/3pxRGBImage
  }
  n14 = Desaturation(None=@n13.RGB, Fraction=ScalarParameter() {group=TimeTrotter, parameterName=PanningColorGradient.Desaturation})
  n15 = NamedRerouteUsage()
  NamedRerouteDeclaration(None=Multiply(A=Multiply(A=@n14, B=@n15), B=ScalarParameter() {defaultValue=1, group=TimeTrotter, parameterName=PanningColorGradient.Brightness})) {name=AdvancedEmissiveOutput, nodeColor=<rgba 1 0 0.2578 1>}
  NamedRerouteDeclaration(None=Multiply(A=@n10.R, B=SmoothStep(Value=@n10.G) {constMax=0.1, constMin=0.02})) {name=EmissiveMask, nodeColor=<rgba 1 0.2578 0 1>}
  Reroute(None=@n7.Tech Art Mask)
 ]
}
